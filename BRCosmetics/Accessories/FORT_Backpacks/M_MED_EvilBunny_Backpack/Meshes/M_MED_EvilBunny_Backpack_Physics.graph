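# Physics asset for the Evil Bunny backpack cosmetic, defining collision and joint limits for its dangling bunny's dynamic bones.
# Capsules cover torso, head, limbs and a three-link chain; boxes cover the backpack root and five-segment ears.
# Constraints branch from the chest to head, arms, spine and chain, with legs off the spine and ears off the head.
# Body joints lock twist, ears and chain use tight limits, and the backpack root is unconstrained.
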PhysicsAsset /BRCosmetics/Accessories/FORT_Backpacks/M_MED_EvilBunny_Backpack/Meshes/M_MED_EvilBunny_Backpack_Physics
bodies {
 backpack_root [{shapeType=Box, extentX=32.3104, extentY=7.6396, extentZ=49.2907, center=<-17.2698 0.5477 -0.0001>, rotation=<rot 0 0.0015 0>}]
 dyn_chain_01 [{shapeType=Capsule, length=0.6425, radius=1.4938, center=<0.2995 0 0.0001>, rotation=<rot -90 174.9489 -174.9488>}]
 dyn_chain_02 [{shapeType=Capsule, length=0.6425, radius=1.4938, center=<0.2992 0 0.014>, rotation=<rot -90 0.0098 -0.0098>}]
 dyn_chain_03 [{shapeType=Capsule, length=0.6425, radius=1.4938, center=<0.2995 0 0.0054>, rotation=<rot -90 0.0098 -0.0098>}]
 dyn_chest [{shapeType=Capsule, length=1.9077, radius=2.8318, center=<-0.7681 -0.0001 0>, rotation=<rot -90 0.0199 -0.0201>}]
 dyn_head [{shapeType=Capsule, length=0.8905, radius=9.6117, center=<5.8454 -1.8612 -0.5473>, rotation=<rot 79.9999 0 0>}]
 dyn_l_arm [{shapeType=Capsule, length=1.5314, radius=1.1737, center=<3.3122 0 0>, rotation=<rot 90 2.5305 2.5305>}]
 dyn_l_ear_01 [{
   shapeType=Box
   extentX=2.3761
   extentY=5.6224
   extentZ=7.6995
   shapeName=dyn_l_ear_01_box
   center=<3.623 1.0607 0.1933>
   rotation=<rot 0 -80.5841 -86.8231>
  }]
 dyn_l_ear_02 [{
   shapeType=Box
   extentX=2.9097
   extentY=6.4598
   extentZ=6.5936
   shapeName=dyn_l_ear_02_box
   center=<3.2957 0.7229 0.0495>
   rotation=<rot 3.9229 -90 4.1429>
  }]
 dyn_l_ear_03 [{
   shapeType=Box
   extentX=1.4656
   extentY=7.6649
   extentZ=8.4027
   shapeName=dyn_l_ear_03_box
   center=<3.714 -0.0844 0.3349>
   rotation=<rot 1.7393 -90 6.297>
  }]
 dyn_l_ear_04 [{
   shapeType=Box
   extentX=2.6785
   extentY=7.8209
   extentZ=9.7383
   shapeName=dyn_l_ear_04_box
   center=<4.5026 0.3473 -0.3008>
   rotation=<rot 15.329 -89.9999 9.98>
  }]
 dyn_l_ear_05 [{
   shapeType=Box
   extentX=2.0909
   extentY=2.998
   extentZ=6.292
   shapeName=dyn_l_ear_05_box
   center=<2.0327 0.0922 3.0858>
   rotation=<rot 0 94.7173 -46.8581>
  }]
 dyn_l_elbow [{shapeType=Capsule, length=2.0493, radius=2.1413, center=<1.9092 0 0>, rotation=<rot 90 0 0>}]
 dyn_l_knee [{shapeType=Capsule, length=4.5054, radius=3.1991, center=<2.4265 0.1396 -0.032>, rotation=<rot -90 -0.0099 0.0099>}]
 dyn_l_leg [{shapeType=Capsule, length=5.0559, radius=2.0203, center=<2.8286 0.0199 0.0027>, rotation=<rot -90 0.0132 -0.0132>}]
 dyn_r_arm [{shapeType=Capsule, length=2.1247, radius=1.1737, center=<4.4971 0 0>, rotation=<rot 90 -1.2659 -1.2659>}]
 dyn_r_ear_01 [{
   shapeType=Box
   extentX=1.8744
   extentY=4.2657
   extentZ=5.7369
   shapeName=dyn_r_ear_01_box
   center=<2.0938 0.3384 -0.0363>
   rotation=<rot 0 97.8863 -89.7221>
  }]
 dyn_r_ear_02 [{
   shapeType=Box
   extentX=2.5556
   extentY=6.6542
   extentZ=5.6566
   shapeName=dyn_r_ear_02_box
   center=<2.9752 0.3248 -0.2253>
   rotation=<rot 2.5525 -89.9995 -4.2757>
  }]
 dyn_r_ear_03 [{
   shapeType=Box
   extentX=2.3648
   extentY=6.1357
   extentZ=7.7174
   shapeName=dyn_r_ear_03_box
   center=<3.2601 0.123 0.1916>
   rotation=<rot 4.878 -89.9995 1.8874>
  }]
 dyn_r_ear_04 [{
   shapeType=Box
   extentX=2.4858
   extentY=5.4814
   extentZ=8.7016
   shapeName=dyn_r_ear_04_box
   center=<3.3121 0.0652 0.2707>
   rotation=<rot 0.6154 -90 4.5601>
  }]
 dyn_r_ear_05 [{
   shapeType=Box
   extentX=2.5659
   extentY=7.1802
   extentZ=11.6405
   shapeName=dyn_r_ear_05_box
   center=<2.1487 0.001 -1.0999>
   rotation=<rot 0 -82.8925 -46.7896>
  }]
 dyn_r_elbow [{shapeType=Capsule, length=1.2103, radius=1.7838, center=<2.7699 0 0>, rotation=<rot 90 0 0>}]
 dyn_r_knee [{shapeType=Capsule, length=4.1537, radius=3.029, center=<2.3937 -0.0159 -0.032>, rotation=<rot -90 -3.7927 3.7927>}]
 dyn_r_leg [{shapeType=Capsule, length=5.0559, radius=2.0203, center=<2.2213 -0.0198 0.0027>, rotation=<rot -90 2.5305 -2.5304>}]
 dyn_spine [{shapeType=Capsule, length=1.3784, radius=4.693, center=<2.83 -0.0001 0>, rotation=<rot -90 -0.0395 0.0394>}]
}
constraints [
 {bone1Name=dyn_l_arm, bone2Name=dyn_chest, swing1LimitDegrees=75, swing2LimitDegrees=15, twistLimitDegrees=45, twistMotion=Locked}
 {bone1Name=dyn_l_elbow, bone2Name=dyn_l_arm, swing1LimitDegrees=45, swing2LimitDegrees=10, twistLimitDegrees=45, twistMotion=Locked}
 {bone1Name=dyn_r_arm, bone2Name=dyn_chest, swing1LimitDegrees=75, swing2LimitDegrees=15, twistLimitDegrees=45, twistMotion=Locked}
 {bone1Name=dyn_r_elbow, bone2Name=dyn_r_arm, swing1LimitDegrees=60, swing2LimitDegrees=10, twistLimitDegrees=45, twistMotion=Locked}
 {bone1Name=dyn_spine, bone2Name=dyn_chest, swing1LimitDegrees=30, swing2LimitDegrees=30, twistLimitDegrees=45, twistMotion=Locked}
 {bone1Name=dyn_l_leg, bone2Name=dyn_spine, swing1LimitDegrees=10, swing2LimitDegrees=30, twistLimitDegrees=45, twistMotion=Locked}
 {bone1Name=dyn_l_knee, bone2Name=dyn_l_leg, swing1LimitDegrees=10, swing2LimitDegrees=30, twistLimitDegrees=45, twistMotion=Locked}
 {bone1Name=dyn_r_leg, bone2Name=dyn_spine, swing1LimitDegrees=10, swing2LimitDegrees=30, twistLimitDegrees=45, twistMotion=Locked}
 {bone1Name=dyn_r_knee, bone2Name=dyn_r_leg, swing1LimitDegrees=10, swing2LimitDegrees=30, twistLimitDegrees=45, twistMotion=Locked}
 {bone1Name=dyn_head, bone2Name=dyn_chest, swing1LimitDegrees=5, swing2LimitDegrees=10, twistLimitDegrees=45, twistMotion=Locked}
 {bone1Name=dyn_l_ear_01, bone2Name=dyn_head, swing1LimitDegrees=10, swing2LimitDegrees=5, twistLimitDegrees=2}
 {bone1Name=dyn_l_ear_02, bone2Name=dyn_l_ear_01, swing1LimitDegrees=10, swing2LimitDegrees=5, twistLimitDegrees=2}
 {bone1Name=dyn_l_ear_03, bone2Name=dyn_l_ear_02, swing1LimitDegrees=15, swing2LimitDegrees=5, twistLimitDegrees=2}
 {bone1Name=dyn_l_ear_04, bone2Name=dyn_l_ear_03, swing1LimitDegrees=15, swing2LimitDegrees=5, twistLimitDegrees=2}
 {bone1Name=dyn_l_ear_05, bone2Name=dyn_l_ear_04, swing1LimitDegrees=15, swing2LimitDegrees=5, twistLimitDegrees=2}
 {bone1Name=dyn_r_ear_01, bone2Name=dyn_head, swing1LimitDegrees=15, swing2LimitDegrees=5, twistLimitDegrees=2}
 {bone1Name=dyn_r_ear_02, bone2Name=dyn_r_ear_01, swing1LimitDegrees=10, swing2LimitDegrees=5, twistLimitDegrees=2}
 {bone1Name=dyn_r_ear_03, bone2Name=dyn_r_ear_02, swing1LimitDegrees=10, swing2LimitDegrees=5, twistLimitDegrees=2}
 {bone1Name=dyn_r_ear_04, bone2Name=dyn_r_ear_03, swing1LimitDegrees=10, swing2LimitDegrees=5, twistLimitDegrees=2}
 {bone1Name=dyn_r_ear_05, bone2Name=dyn_r_ear_04, swing1LimitDegrees=10, swing2LimitDegrees=5, twistLimitDegrees=2}
 {bone1Name=dyn_chain_01, bone2Name=dyn_chest, swing1LimitDegrees=15, swing2LimitDegrees=2, twistLimitDegrees=2}
 {bone1Name=dyn_chain_02, bone2Name=dyn_chain_01, swing1LimitDegrees=15, swing2LimitDegrees=2, twistLimitDegrees=2}
 {bone1Name=dyn_chain_03, bone2Name=dyn_chain_02, swing1LimitDegrees=15, swing2LimitDegrees=2, twistLimitDegrees=2}
]
